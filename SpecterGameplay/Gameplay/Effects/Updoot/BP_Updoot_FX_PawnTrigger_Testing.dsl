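; Test trigger actor for Updoot effects that toggles three Niagara systems on a FortPawn entering its box.
; Each overlap flips a FlipFlop between adding and removing through AddRemoveFX, which runs only where cosmetic events can execute.
; Adding spawns the systems attached to the pawn's mesh, or reattaches and activates the cached components if the first exists.
; Removing deactivates and detaches them for reuse.
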(event EventBeginPlay)

(event EventTick
  (params (DeltaSeconds :type Float%20%28single-precision%29)))

(event OnComponentBeginOverlap(Box)
  (params (OverlappedComponent :type Primitive%20Component%20Object%20Reference) (OtherActor :type Actor%20Object%20Reference) (OtherComp :type Primitive%20Component%20Object%20Reference) (OtherBodyIndex :type Integer) (bFromSweep :type Boolean) (SweepResult :type Hit%20Result%20Structure%20%28by%20ref%29))
  (bind _asfort_pawn (Utilities|Casting|CastToFortPawn OtherActor))
  (bind _isa (Utilities|FlowControl|FlipFlop)
    (:A
      _)
    (:B
      _))
  (CallFunction|AddRemoveFX _isa _asfort_pawn))

(event Custom|AddRemoveFX
  (params (Add :type Boolean) (Pawn :type Fort%20Pawn%20Object%20Reference))
  (bind _mesh (Variables|/Script/Engine.Character|GetMesh Pawn))
  (bind _spawned_system2 (Variables|Default|Getspawned%20system2))
  (bind _spawned_system2_2 (Variables|Default|Getspawned%20system2))
  (bind _spawned_system3 (Variables|Default|Getspawned%20system3))
  (bind _spawned_system3_2 (Variables|Default|Getspawned%20system3))
  (Networking|CanExecuteCosmeticEvents
    (:True
      (Utilities|IsValid Pawn
        (:Is%20Valid
          (if Add
            (bind _spawnedsystem (Variables|Default|GetSpawnedSystem)
              (:then
                (Transformation|AttachComponentToComponent _spawnedsystem _mesh "None" "KeepRelative" "KeepRelative" "KeepRelative" true)
                (Components|Activation|Activate _spawnedsystem false)
                (Transformation|AttachComponentToComponent _spawned_system2 _mesh "None" "KeepRelative" "KeepRelative" "KeepRelative" true)
                (Components|Activation|Activate _spawned_system2 false)
                (Transformation|AttachComponentToComponent _spawned_system3 _mesh "None" "KeepRelative" "KeepRelative" "KeepRelative" true)
                (Components|Activation|Activate _spawned_system3 false))
              (:else
                (bind _returnvalue (Niagara|SpawnSystemAttached (Variables|Default|GetParticles) _mesh))
                (Variables|Default|SetSpawnedSystem _returnvalue)
                (bind _returnvalue_2 (Niagara|SpawnSystemAttached (Variables|Default|GetParticles_0) _mesh))
                (Variables|Default|Setspawned%20system2 _returnvalue_2)
                (bind _returnvalue_3 (Niagara|SpawnSystemAttached (Variables|Default|GetParticles_1) _mesh))
                (Variables|Default|Setspawned%20system3 _returnvalue_3)))
            (else
              (bind _spawnedsystem_2 (Variables|Default|GetSpawnedSystem)
                (:then
                  (Components|Activation|Deactivate _spawnedsystem_2)
                  (Transformation|DetachFromComponent _spawnedsystem_2 "KeepRelative" "KeepRelative" "KeepRelative" true)
                  (Components|Activation|Deactivate _spawned_system2_2)
                  (Transformation|DetachFromComponent _spawned_system2_2 "KeepRelative" "KeepRelative" "KeepRelative" true)
                  (Components|Activation|Deactivate _spawned_system3_2)
                  (Transformation|DetachFromComponent _spawned_system3_2 "KeepRelative" "KeepRelative" "KeepRelative" true))
                (:else)))))))))

(fn ConstructionScript)
