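# Unlit translucent Niagara sprite material for the Updoot effect's target lock-on reticle, drawn without depth testing.
# The MainTex state builds a half-strength fill plus outline from a channel-packed distance field, while the LockedIcon state tints its red channel with PipColor.
# Dynamic parameter TargetLocked blends color and opacity between states, scaled by particle color and alpha, with color passing through an emissive adjustment function.
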Material /SpecterGameplay/Gameplay/Effects/Updoot/M_Updoot_Target
graph {
 properties {
  bDisableDepthTest=true
  bUseTranslucencyVertexFog=false
  bUsedWithNiagaraSprites=true
  blendMode=BLEND_Translucent
  pixelDepthOffsetMode=PDOM_Legacy
  shadingModel=MSM_Unlit
  translucencyPass=MTP_AfterMotionBlur
  twoSided=true
  usedShadingModels=Unlit
 }
 nodes [
  n1 = NamedRerouteUsage()
  n2 = NamedRerouteUsage()
  n3 = DynamicParameter() {defaultValue=<rgba 0 1 1 1>, paramNames [TargetLocked,Param2,Param3,Param4]}
  n4 = ParticleColor()
  MaterialFunctionCall(Emissive=Multiply(A=LinearInterpolate(A=@n1, B=@n2, Alpha=@n3.TargetLocked) {constAlpha=1}, B=@n4.RGB)) {materialFunction=&MF_FX_EmissiveAdjustments}
  Multiply(A=@n4.A, B=LinearInterpolate(A=NamedRerouteUsage(), B=NamedRerouteUsage(), Alpha=@n3.TargetLocked))
  n5 = TextureSampleParameter2D() {
   parameterName=LockedIcon
   samplerSource=SSM_Clamp_WorldGroupSettings
   samplerType=SAMPLERTYPE_LinearColor
   texture=/SpecterGameplay/Reticle/Textures/T_Reticle_SpecterUpdoot_LockOn
  }
  NamedRerouteDeclaration(None=Multiply(A=@n5.R, B=VectorParameter() {parameterName=PipColor, defaultValue=<rgba 1 0.0126 0.7957 1>}.RGB)) {name=LockedOnColor, nodeColor=<rgba 0 0 0 1>}
  NamedRerouteDeclaration(None=@n5.R) {name=LockedOnOpacity, nodeColor=<rgba 0 0 0 1>}
  n6 = TextureSampleParameter2D() {
   parameterName=MainTex
   samplerSource=SSM_Clamp_WorldGroupSettings
   samplerType=SAMPLERTYPE_LinearColor
   texture=&T_Reticle_SpecterUpdoot_LockOn
  }
  n7 = MaterialFunctionCall!(RGBA Channels=@n6.R, Glow Min, Glow Max, Outline Thickness=Constant() {r=0.8}) {materialFunction=/FortUILibrary/MaterialFunctions/Basics/MF_UI_ChannelPackedDistanceField}
  n8 = NamedRerouteDeclaration(None=@n7.Fill) {name=Fill, nodeColor=<rgba 0 0 0 1>}
  n9 = Constant() {r=0.5}
  NamedRerouteDeclaration(None=Add(A=Multiply(A=@n8, B=@n9), B=NamedRerouteDeclaration(None=@n7.Outline) {name=Outline, nodeColor=<rgba 0 0 0 1>})) {name=Opacity, nodeColor=<rgba 0 0 0 1>}
  n10 = NamedRerouteUsage()
  n11 = Constant3Vector()
  n12 = NamedRerouteUsage()
  NamedRerouteDeclaration(None=LinearInterpolate(A=Multiply(A=@n10, B=@n11), B=@n12, Alpha=@n12)) {name=Color, nodeColor=<rgba 0 0 0 1>}
 ]
}
