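# Niagara camera lens effect that, judging by its name, flashes a brief teal burst for a Ghost Updoot event.
# A single emitter spawns one short-lived, camera-offset particle after a slight delay, drawn through two mesh renderers.
# Curves pulse a dynamic material parameter and shape a quick alpha fade-in and fade-out.
# A time-of-day select further scales alpha, strongest by day and weakest at night.
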NiagaraSystem /SpecterGameplay/Gameplay/Effects/Updoot/NS_Camera_Lens_Ghost_Updoot_Burst
dependencies {usedDataInterfaces [&NiagaraDataInterfaceCurve], usedDynamicInputs [], usedModules [&SystemState]}
emitters {GhostUpdootBurstLensEffect {
  emitterUpdateScript [
   &EmitterState(Inactive Response=::NewEnumerator1, Life Cycle Mode=::NewEnumerator1, Loop Behavior=::NewEnumerator1, Loop Delay=0.05, UseLoopDelay=true)
   &SpawnBurst_Instantaneous(Spawn Count=1, Use Spawn Probability=false)
  ]
  particleSpawnScript [
   &CameraOffset(Camera Offset Amount=-30) {enabled=false}
   &InitializeParticle(
    Color=<rgba 0 0.4635 0.2997 0.913>
    Lifetime=0.25
    Mass Mode=::NewEnumerator1
    Mesh Scale=<100 100 100>
    Mesh Scale Mode=::NewEnumerator4
    Position Offset=<-250 0 0>
    Sprite Size Mode=::NewEnumerator4
    UsePositionOffset=false
    Write Material Random=true
   ) {module=InitializeParticle001}
  ]
  particleUpdateScript [
   &ParticleState()
   &SolveForcesAndVelocity(Clamp Velocity=false, Limit Acceleration=false)
   &ScaleColor(
    Scale Alpha=&FloatFromCurve(
     FloatCurve=DataInterface() {propertyValues {Curve {
        defaultValue=0.5029
        keys [
         {interpMode=RCIM_Cubic, tangentMode=RCTM_User, time=0.0467, value=0.5917}
         {interpMode=RCIM_Cubic, tangentMode=RCTM_User, time=0.5023, value=0.6665}
         {tangentMode=RCTM_User, time=1.0467, value=0.5917}
        ]
       }}}
     Scale Curve=0.075
    )
    Scale RGB=&Lerp_Vector(
     A=&VectorFromFloat(Value=4)
     B=&VectorFromFloat(Value=2)
     Alpha=&MakeFloatFromVector(
      Vector=&VectorFromFloat(Value=&Waveform("Waveform [1]"=::NewEnumerator0, "[1] Frequency"=0.15, "[1] Amplitude Min/Max"=<0 1>, "[1] Bias"=0.5))
      Channel=::NewEnumerator0
     )
    )
    ScaleA=true
    ScaleRGB=true
   ) {enabled=false, module=ScaleColor002}
   &DynamicMaterialParameters(
    Index 0 Param 2=&FloatFromCurve(
     FloatCurve=DataInterface() {propertyValues {Curve {
        defaultValue=1
        keys [
         {interpMode=RCIM_Cubic, tangentMode=RCTM_User, time=0, value=0}
         {interpMode=RCIM_Cubic, tangentMode=RCTM_User, time=0.5, value=1}
         {tangentMode=RCTM_User, time=1, value=0}
        ]
       }}}
     Scale Curve=0.35
    )
    Index 0 Param 3=0.9
    Module.Index0Float0Write=true
    Module.Index0Float1Write=true
    Module.Index0Float2Write=true
    Module.Index0Float3Write=true
   )
   &ScaleColor(
    Scale Alpha=&FloatFromCurve(FloatCurve=DataInterface() {propertyValues {Curve {
        defaultValue=1
        keys [
         {arriveTangent=9.0725, interpMode=RCIM_Cubic, leaveTangent=9.0725, tangentMode=RCTM_User, time=0, value=0}
         {arriveTangent=-2.5249, interpMode=RCIM_Cubic, leaveTangent=-2.5249, tangentMode=RCTM_User, time=0.1408, value=0.9847}
         {arriveTangent=-0.0405, interpMode=RCIM_Cubic, leaveTangent=-0.0405, tangentMode=RCTM_User, time=1, value=0}
        ]
       }}})
    ScaleA=true
    ScaleRGB=false
   ) {module=ScaleColor001}
   &ScaleColor(
    Scale Alpha=&FN_Select_Float_from_Time_Of_Day(Morning=0.6, Day=0.8, Evening=0.6, Night=0.4)
    ScaleA=true
    ScaleRGB=false
   )
   &CameraOffset(Camera Offset Amount=1) {module=CameraOffset001}
  ]
  rendererClasses [&NiagaraMeshRendererProperties]
  renderers [
   {rendererClass=&NiagaraMeshRendererProperties, rendererIndex=0}
   {rendererClass=&NiagaraMeshRendererProperties, rendererIndex=1}
  ]
 }}
summary {
 systemName=NS_Camera_Lens_Ghost_Updoot_Burst
 emitters [{emitterName=GhostUpdootBurstLensEffect, rendererClasses [&NiagaraMeshRendererProperties]}]
 userVariables []
}
