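# Niagara projectile trail for the Updoot launcher, drawing a cyan-green projectile body and the trail behind it.
# Distance-based spawning feeds a short-lived ribbon and puffy smoke sprites, both animated through dynamic material parameters.
# Updoot_Sprite bursts a looping mesh particle with squash scaling, waveform-driven arm wiggle in the material, camera-distance upscaling and time-of-day color.
# The Updoot_Arrow emitter, an alternate sprite, mesh and light body, is disabled.
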NiagaraSystem /SpecterGameplay/Gameplay/Effects/Updoot/NS_Updoot_Launcher_PrjTrail
dependencies {
 usedDataInterfaces [
  &NiagaraDataInterfaceArrayColor
  &NiagaraDataInterfaceColorCurve
  &NiagaraDataInterfaceCurve
  &NiagaraDataInterfaceArrayFloat
 ]
 usedDynamicInputs []
 usedModules [&SystemState]
}
emitters {
 Rocket_Trail_Puffy_Smoke {
  emitterUpdateScript [
   &EmitterState(Life Cycle Mode=::NewEnumerator1, Loop Duration=0.5, UseLoopDelay=false)
   &SpawnPerUnit(Module.Use Max Movement Threshold=false, Module.Use Movement Tolerance=false, Spawn Spacing=40, Use Spawn Probability=false)
  ]
  particleSpawnScript [
   &InitializeParticle(
    Color=&SelectLinearColorFromArray(
     Array Sampling Mode=::NewEnumerator0
     Direct Array Index=?
     Color Selection Array=DataInterface() {propertyValues {ColorData [<rgba 0 0.7302 1 0.1043>, <rgba 0 0.6692 1.15 0.1391>, <rgba 0 0.9479 0.59 0.513>]}}
     Random Seed=0
    )
    Lifetime Max=0.5
    Lifetime Min=0.25
    Lifetime Mode=::NewEnumerator1
    Sprite Rotation Mode=::NewEnumerator2
    Sprite Size Mode=::NewEnumerator1
    Sprite UV Mode=::NewEnumerator3
    Uniform Sprite Size Max=180
    Uniform Sprite Size Min=120
    UsePositionOffset=false
    Write Material Random=true
   )
   &InheritVelocity(Inherited Velocity Amount Scale=<0.1 0.1 0.1>, Inherited Velocity Speed Limit=5000, Speed Limit=true, UseSpeedThreshold=false)
  ]
  particleUpdateScript [
   &SolveForcesAndVelocity(Clamp Velocity=false, Limit Acceleration=false)
   &ParticleState()
   &ScaleSpriteSize(
    Uniform Curve Scale=0.5
    Uniform Curve Sprite Scale=DataInterface() {propertyValues {Curve {keys [{interpMode=RCIM_Cubic, leaveTangent=2.3279, tangentMode=RCTM_Break, time=0, value=0}, {tangentMode=RCTM_User, time=1, value=1}]}}}
   )
   &ScaleColor(
    Linear Color Curve=DataInterface() {propertyValues {
      AlphaCurve {keys [<at 0 0.7>,<at 1 0>]}
      BlueCurve {keys [<at 0 0>,<at 0.107 1>]}
      GreenCurve {keys [<at 0 8.4435>,<at 0.107 1>]}
      RedCurve {keys [<at 0 50>,<at 0.107 1>]}
     }}
    Scale Mode=::NewEnumerator2
   ) {enabled=false}
   &ScaleColor(
    Scale Alpha=&FloatFromCurve(FloatCurve=DataInterface() {propertyValues {Curve {keys [<at 0 0>,<at 0.2049 0.9765>,<at 0.7791 0.9107>,<at 1 0>]}}})
    ScaleA=true
    ScaleRGB=false
   ) {enabled=false, module=ScaleColor001}
   &ScaleColor(
    Scale Alpha=&FloatFromCurve(FloatCurve=DataInterface() {propertyValues {Curve {keys [<at -0.0046 1.0113 RCIM_Cubic>,<at 1.0011 0.002 RCIM_Cubic>]}}})
    ScaleA=true
    ScaleRGB=true
   ) {enabled=false, module=ScaleColor002}
   &DynamicMaterialParameters(
    Index 0 Param 1=&FloatFromCurve(
     FloatCurve=DataInterface() {propertyValues {Curve {keys [
         {interpMode=RCIM_Cubic, leaveTangent=15.2028, tangentMode=RCTM_Break, time=0, value=1}
         {interpMode=RCIM_Cubic, leaveTangent=10.7187, tangentMode=RCTM_Break, time=1, value=6}
        ]}}}
     Scale Curve=0
    )
    Index 0 Param 3=&FloatFromCurve(FloatCurve=DataInterface() {propertyValues {Curve {keys [{interpMode=RCIM_Cubic, tangentMode=RCTM_Break, time=0, value=1}, <at 1 -1>]}}})
    Index 0 Param 4=1
    Module.Index0Param0Write=true
    Module.Index0Param1Write=true
    Module.Index0Param2Write=true
    Module.Index0Param3Write=true
   ) {enabled=false}
   &DynamicMaterialParameters(
    Index 0 Param 1=&FloatFromCurve(FloatCurve=DataInterface() {propertyValues {Curve {keys [<at 0 0 RCIM_Cubic>,<at 1.0011 0.2>]}}})
    Index 0 Param 2=&FloatFromCurve(FloatCurve=DataInterface() {propertyValues {Curve {keys [
         {arriveTangent=0.0453, interpMode=RCIM_Cubic, tangentMode=RCTM_Break, time=0, value=-1}
         {arriveTangent=0.0604, interpMode=RCIM_Cubic, leaveTangent=0.0604, tangentMode=RCTM_User, time=1, value=-1}
        ]}}})
    Index 0 Param 3=&FloatFromCurve(FloatCurve=DataInterface() {propertyValues {Curve {keys [{interpMode=RCIM_Cubic, leaveTangent=4.6779, time=0.2851, value=0}]}}})
    Index 1 Param 1=&FloatFromCurve(
     FloatCurve=DataInterface() {propertyValues {Curve {keys [<at 0 0>,<at 0.25 1>]}}}
     Scale Curve=&RandomRangeFloat(Minimum=2, Maximum=1)
    )
    Index 1 Param 2=&FloatFromCurve(
     FloatCurve=DataInterface() {propertyValues {Curve {keys [
         {interpMode=RCIM_Cubic, leaveTangent=3.7422, tangentMode=RCTM_Break, time=0, value=0.2}
         {interpMode=RCIM_Cubic, leaveTangent=0.9066, time=1, value=2}
        ]}}}
     Scale Curve=&RandomRangeFloat(Minimum=0.5, Maximum=1)
    )
    Index 1 Param 3=&FloatFromCurve(FloatCurve=DataInterface() {propertyValues {Curve {keys [
         {interpMode=RCIM_Cubic, leaveTangent=2.8445, tangentMode=RCTM_Break, time=0, value=0.2}
         {interpMode=RCIM_Cubic, leaveTangent=63.2978, time=0.25, value=1}
        ]}}})
    Index 1 Param 4=&FloatFromCurve(
     FloatCurve=DataInterface() {propertyValues {Curve {keys [<at 0 1>,<at 1 0.8>]}}}
     Scale Curve=&RandomRangeFloat(Minimum=0.5, Maximum=1)
    )
    Index 2 Param 1=&FloatFromCurve(FloatCurve=DataInterface() {propertyValues {Curve {keys [{interpMode=RCIM_Cubic, leaveTangent=3.2309, tangentMode=RCTM_Break, time=0.1, value=0}, <at 1 1>]}}})
    Index 2 Param 2=$Particles.MaterialRandom
    Index 2 Param 3=&FloatFromCurve(FloatCurve=DataInterface() {propertyValues {Curve {keys [
         {interpMode=RCIM_Cubic, tangentMode=RCTM_Break, time=0, value=1}
         {interpMode=RCIM_Cubic, leaveTangent=11.1233, tangentMode=RCTM_Break, time=1, value=2}
        ]}}})
    Index 2 Param 4=0.75
    Module.Index0Float0Write=true
    Module.Index0Float1Write=true
    Module.Index0Float2Write=true
    Module.Index0Float3Write=true
    Module.Index1Float0Write=true
    Module.Index1Float1Write=true
    Module.Index1Float2Write=true
    Module.Index1Float3Write=true
    Module.Index2Float0Write=true
    Module.Index2Float1Write=true
    Module.Index2Float2Write=true
    Module.Index2Float3Write=true
    Write Parameter Index 1=::NewEnumerator1
    Write Parameter Index 2=::NewEnumerator1
   ) {module=DynamicMaterialParameters001}
  ]
  rendererClasses [&NiagaraSpriteRendererProperties]
  renderers [{rendererClass=&NiagaraSpriteRendererProperties}]
 }
 Rocket_Trail_Ribbon {
  emitterUpdateScript [
   &EmitterState()
   &SpawnPerUnit(Module.Use Max Movement Threshold=true, Module.Use Movement Tolerance=false, Use Spawn Probability=false)
   &SpawnRate(SpawnRate=20, Use Spawn Probability=false) {enabled=false}
  ]
  particleSpawnScript [&InitializeParticle(
    Color=&SelectLinearColorFromArray(
     Array Sampling Mode=::NewEnumerator0
     Direct Array Index=?
     Color Selection Array=DataInterface() {propertyValues {ColorData [<rgba 0 0.7302 1 0.0957>, <rgba 0 0.6692 1.15 0.1391>, <rgba 0 0.9479 0.59 0.113>]}}
     Random Seed=0
    )
    Lifetime=0.25
    Ribbon Width=25
    Ribbon Width Mode=::NewEnumerator1
    UsePositionOffset=false
    Write Material Random=true
   )]
  particleUpdateScript [
   &ParticleState()
   &ScaleColor(
    Scale Mode=::NewEnumerator1
    Scale RGBA=&Multiply_Vector4(
     A=&MakeVector4_FromLinearColor(LinearColor=&ColorFromCurve(ColorCurve=DataInterface() {propertyValues {
         AlphaCurve {keys [<at 0 0>,<at 0.2284 1>,<at 1 0>]}
         BlueCurve {keys [<at 0 0.773>,<at 1 0.5417>]}
         GreenCurve {keys [<at 0 2>,<at 1 0.4482>]}
         RedCurve {keys [<at 0 0>,<at 1 0>]}
        }}))
     B=&MakeVector4FromVectorAndFloat("Vector (RGB)"=<1 1 1>, "Float (Alpha)"=0.85)
    )
    ScaleRGBA=true
   )
   &ScaleColor(
    Scale Mode=::NewEnumerator1
    Scale RGBA=&Multiply_Vector4(
     A=&MakeVector4_FromLinearColor(LinearColor=&ColorFromCurve(ColorCurve=DataInterface() {propertyValues {
         AlphaCurve {keys [<at 0 1>,<at 1 0>]}
         BlueCurve {keys [<at 0 20>,<at 0.9981 20>]}
         GreenCurve {keys [<at 0 20>,<at 0.9981 20>]}
         RedCurve {keys [<at 0 20>,<at 0.9981 20>]}
        }}))
     B=&MakeVector4FromVectorAndFloat(
      "Vector (RGB)"=<1 1 1>
      "Float (Alpha)"=&FloatFromCurve(FloatCurve=DataInterface() {propertyValues {Curve {keys [<at 0 0>,<at 0.5 1>]}}}, CurveIndex=$Engine.System.Age, Scale Curve=0.5)
     )
    )
    ScaleRGBA=false
   ) {module=ScaleColor002}
   &DynamicMaterialParameters(
    Index 0 Param 1=&FloatFromCurve(FloatCurve=DataInterface() {propertyValues {Curve {keys [{interpMode=RCIM_Cubic, tangentMode=RCTM_Break, time=0, value=1.35}, <at 1 0>]}}})
    Index 0 Param 2=1
    Index 0 Param 3=1
    Index 0 Param 4=1
    Module.Index0Param0Write=true
    Module.Index0Param1Write=true
    Module.Index0Param2Write=true
    Module.Index0Param3Write=true
   )
   &ScaleRibbonWidth(Ribbon Width Scale=&FloatFromCurve(
     FloatCurve=DataInterface() {propertyValues {Curve {keys [{interpMode=RCIM_Cubic, tangentMode=RCTM_Break, time=0, value=1}, <at 1 2>]}}}
     Scale Curve=0.5
    ))
   &ScaleColor(Scale RGB=&VectorFromFloat(Value=1), ScaleA=true, ScaleRGB=true) {enabled=false, module=ScaleColor001}
  ]
  rendererClasses [&NiagaraRibbonRendererProperties]
  renderers [{rendererClass=&NiagaraRibbonRendererProperties}]
 }
 Updoot_Arrow {
  bEnabled=false
  emitterUpdateScript [
   /SetVariables(Emitter.TODColorMult:LinearColor=&FNTimeOfDayColor(Night Color=<rgba 1 1 1 1>, Morning Color=<rgba 1.25 1.25 1.25 1>, Day Color=<rgba 1.5 1.5 1.5 1>, Evening Color=<rgba 1.25 1.25 1.25 1>)) {isSetParameters=true, module=SetVariables_#1}
   &EmitterState(
    Inactive Response=::NewEnumerator1
    Life Cycle Mode=::NewEnumerator1
    Loop Behavior=::NewEnumerator1
    Loop Duration Mode=::NewEnumerator1
    UseLoopDelay=false
   )
   &SpawnBurst_Instantaneous(Use Spawn Probability=false, UseLoopCountLimit=false)
   /SetVariables(
    Emitter.LightColor:LinearColor=&ColorFromCurve(
     ColorCurve=DataInterface() {propertyValues {
       AlphaCurve {keys [<at 0.0513 1 RCIM_Cubic>]}
       BlueCurve {keys [<at 0.084 0>,<at 0.2843 0 RCIM_Cubic>]}
       GreenCurve {keys [<at 0.084 0>,<at 0.2843 3.7041 RCIM_Cubic>]}
       RedCurve {keys [<at 0.084 0>,<at 0.2843 25 RCIM_Cubic>]}
      }}
     CurveIndex=$Emitter.Age
    )
    Emitter.LightExponent=1.5
    Emitter.LightRadius:NiagaraFloat=&Lerp_Float(
     A=0
     B=180
     Alpha=&FloatFromCurve(
      FloatCurve=DataInterface() {propertyValues {Curve {keys [{interpMode=RCIM_Cubic, leaveTangent=1.7761, tangentMode=RCTM_Break, time=0.1, value=0}, <at 1 1>]}}}
      CurveIndex=$Emitter.Age
     )
    )
   ) {isSetParameters=true, module=SetVariables_#2}
  ]
  particleSpawnScript [
   &PartitionParticles()
   &InitializeParticle(
    Color=<rgba 0 1.5 0.9261 1>
    Lifetime=0.17
    Position Offset=<0 1 0>
    Sprite Size Mode=::NewEnumerator3
    Uniform Sprite Size=45
    UsePositionOffset=true
    Write Material Random=true
   )
   &CameraOffset()
  ]
  particleUpdateScript [
   &ParticleState(Kill Particles When Lifetime Has Elapsed=false, Let Infinitely Lived Particles Die When Emitter Deactivates=true)
   /Niagara/Modules/Update/Camera/MaintainInCameraParticleScale(
    Pixel Depth:NiagaraFloat=&Add_Float(A=$Engine.Owner.LODDistance, B=100)
    Render Target Width:NiagaraFloat=&Multiply_Float(A=2560, B=1)
   ) {module=MaintainInCameraParticleScale001}
   &FN_Sprite_Attributes(
    Camera Offset=50
    Color=&MultiplyLinearColor(
     Linear Color=&MultiplyLinearColor(
      Linear Color=&ColorFromCurve(ColorCurve=DataInterface() {propertyValues {
         AlphaCurve {keys [<at 0.0609 0 RCIM_Cubic>,<at 0.6127 0.3 RCIM_Cubic>]}
         BlueCurve {keys [<at 0.0767 0 RCIM_Cubic>,<at 0.5465 0 RCIM_Cubic>]}
         GreenCurve {keys [<at 0.0767 0.0333 RCIM_Cubic>,<at 0.5465 0.1 RCIM_Cubic>]}
         RedCurve {keys [<at 0.0767 1 RCIM_Cubic>,<at 0.5465 1 RCIM_Cubic>]}
        }})
      Scale Factor=&MakeVector4FromVectorAndFloat("Vector (RGB)"=&VectorFromFloat(Value=5), "Float (Alpha)"=1)
     )
     Scale Factor=&MakeVector4_FromLinearColor(LinearColor=$Emitter.TODColorMult)
    )
    Create Camera Offset Attribute=true
    Create Color Attribute=true
    "Create Normalized Age "=true
    Create SpriteSize Attribute=true
    SpriteSize=&Multiply_Vector2D_ByFloat(
     Float=&Multiply_Float(
      A=&FloatFromCurve(FloatCurve=DataInterface() {propertyValues {Curve {keys [
           {interpMode=RCIM_Cubic, leaveTangent=14.8021, tangentMode=RCTM_Break, time=0, value=0}
           {arriveTangent=-5.4057, interpMode=RCIM_Cubic, tangentMode=RCTM_Break, time=0.5, value=5}
          ]}}})
      B=&FN_LOD_Distance_Remap(Start Distance=0, End Distance=1000, Minimum Output=0.1, Maximum Output=2, Clamp Output=true)
     )
     Vector2D=<150 150>
    )
    Write to Module Namespace=true
   )
   &FN_Light_Attributes(
    Color=&MultiplyLinearColor(
     Linear Color=<rgba 1 0.3872 0 1>
     Scale Factor=&MakeVector4_FromLinearColor(LinearColor=$Emitter.TODColorMult)
    )
    Create Color Attribute=true
    Create Exponent Attribute=true
    Create Radius Attribute=true
    Enable Particle Light Enable=true
    Exponent=5
    Light Enabled=true
    Radius=500
    Scale Color=5
   )
   &FN_Mesh_Attributes(
    Color=&MultiplyLinearColor(
     Linear Color=<rgba 0.0094 4 2.1307 0.1043>
     Scale Factor=&MakeVector4_FromLinearColor(LinearColor=$Emitter.TODColorMult)
    )
    Create Color Attribute=true
    Create Orientation Attribute=true
    Create Scale Attribute=true
    Orientation:Quat4f=&MultiplyQuaternion(
     Quaternion A=<0 0 0 1>
     Quaternion B=&MakeQuaternion(Quaternion From=0, Axis=<0 0 1>, Angle=-0.25, From=?, To=?, Angle Type=::NewEnumerator1, Normalize Vectors=false)
    )
    Scale=&Multiply_VectorByFloat(Vector=<2 1 1>, Float=0.5)
    Write to Module Namespace=true
   )
  ]
  rendererClasses [
   &NiagaraSpriteRendererProperties
   &NiagaraLightRendererProperties
   &NiagaraMeshRendererProperties
  ]
  renderers [
   {rendererClass=&NiagaraSpriteRendererProperties}
   {rendererClass=&NiagaraLightRendererProperties, rendererIndex=1}
   {rendererClass=&NiagaraMeshRendererProperties, rendererIndex=2}
   {rendererClass=&NiagaraSpriteRendererProperties, rendererIndex=3}
  ]
 }
 Updoot_Sprite {
  emitterUpdateScript [
   /SetVariables(Emitter.TODColorMult:LinearColor=&FNTimeOfDayColor(Night Color=<rgba 1 1 1 1>, Morning Color=<rgba 1.25 1.25 1.25 1>, Day Color=<rgba 1.5 1.5 1.5 1>, Evening Color=<rgba 1.25 1.25 1.25 1>)) {isSetParameters=true, module=SetVariables_#1}
   &EmitterState(
    Inactive Response=::NewEnumerator1
    Life Cycle Mode=::NewEnumerator1
    Loop Behavior=::NewEnumerator1
    Loop Duration Mode=::NewEnumerator1
    UseLoopDelay=false
   )
   &SpawnBurst_Instantaneous(Use Spawn Probability=false, UseLoopCountLimit=false)
   /SetVariables(
    Emitter.LightColor:LinearColor=&ColorFromCurve(
     ColorCurve=DataInterface() {propertyValues {
       AlphaCurve {keys [<at 0.0513 1 RCIM_Cubic>]}
       BlueCurve {keys [<at 0.084 0>,<at 0.2843 0 RCIM_Cubic>]}
       GreenCurve {keys [<at 0.084 0>,<at 0.2843 3.7041 RCIM_Cubic>]}
       RedCurve {keys [<at 0.084 0>,<at 0.2843 25 RCIM_Cubic>]}
      }}
     CurveIndex=$Emitter.Age
    )
    Emitter.LightExponent=1.5
    Emitter.LightRadius:NiagaraFloat=&Lerp_Float(
     A=0
     B=180
     Alpha=&FloatFromCurve(
      FloatCurve=DataInterface() {propertyValues {Curve {keys [{interpMode=RCIM_Cubic, leaveTangent=1.7761, tangentMode=RCTM_Break, time=0.1, value=0}, <at 1 1>]}}}
      CurveIndex=$Emitter.Age
     )
    )
   ) {isSetParameters=true, module=SetVariables_#2}
  ]
  particleSpawnScript [
   &PartitionParticles()
   &InitializeParticle(
    Color=<rgba 0 1.5 0.9261 1>
    Lifetime=0.4
    Mesh Scale Mode=::NewEnumerator3
    Mesh Uniform Scale=0.4
    Position Offset=<0 1 0>
    Sprite Size Mode=::NewEnumerator3
    Uniform Sprite Size=45
    UsePositionOffset=true
    Write Material Random=true
   )
   &CameraOffset()
   &InitialMeshOrientation(Mesh Orientation Mode=::NewEnumerator2, Use Rotation Vector=false)
   &InheritVelocity(Speed Limit=true, UseSpeedThreshold=false) {enabled=false}
   &AddVelocity(Velocity=<1 0 0>) {enabled=false}
   &ApplyInitialForces() {enabled=false}
  ]
  particleUpdateScript [
   &ParticleState(
    Kill Particles When Lifetime Has Elapsed=false
    Let Infinitely Lived Particles Die When Emitter Deactivates=true
    Loop Particles Lifetime=true
   )
   &UpdateMeshOrientation(
    Facing Direction=$Engine.Owner.Velocity
    Facing Interpolation=false
    Orientation Method=::NewEnumerator1
    Override Quat=false
    TransformFacing=false
    TransformUp=false
   ) {enabled=false}
   &ScaleMeshSize(Scale Factor=&MakeVector(
     X=&FloatFromCurve(FloatCurve=DataInterface() {propertyValues {Curve {
         defaultValue=1
         keys [
          {interpMode=RCIM_Cubic, tangentMode=RCTM_User, time=0, value=1}
          {interpMode=RCIM_Cubic, tangentMode=RCTM_User, time=0.5098, value=0.6311}
          {tangentMode=RCTM_User, time=1, value=1}
         ]
        }}})
     Y=1
     Z=&FloatFromCurve(FloatCurve=DataInterface() {propertyValues {Curve {
         defaultValue=1
         keys [
          {interpMode=RCIM_Cubic, tangentMode=RCTM_User, time=0, value=1}
          {interpMode=RCIM_Cubic, tangentMode=RCTM_User, time=0.5046, value=1.0498}
          {tangentMode=RCTM_User, time=1, value=1}
         ]
        }}})
    ))
   &ScaleMeshSize(Scale Factor=&VectorFromFloat(Value=&FN_LOD_Distance_Remap(Start Distance=250, End Distance=10000, Minimum Output=1, Maximum Output=4, Clamp Output=true))) {module=ScaleMeshSize004}
   /SetVariables(Particles.ArmWiggleFrequency=4) {isSetParameters=true}
   &DynamicMaterialParameters(
    Index 0 Param 2=&Lerp_Float(
     A=-0.25
     B=0
     Alpha=&Waveform("Waveform [1]"=::NewEnumerator0, "[1] Frequency"=3, "[1] Amplitude Min/Max"=<0 1>)
    )
    Index 1 Param 1=&Lerp_Float(
     A=-0.25
     B=0.25
     Alpha=&Waveform("Waveform [1]"=::NewEnumerator0, "[1] Frequency"=$Particles.ArmWiggleFrequency, "[1] Amplitude Min/Max"=<0 1>)
    )
    Index 1 Param 2=&Lerp_Float(
     A=-0.25
     B=0.25
     Alpha=&Waveform("Waveform [1]"=::NewEnumerator0, "[1] Frequency"=$Particles.ArmWiggleFrequency, "[1] Amplitude Min/Max"=<0 1>)
    )
    Index 1 Param 3=&Lerp_Float(
     A=-0.5
     B=0.5
     Alpha=&Waveform("Waveform [1]"=::NewEnumerator0, "[1] Frequency"=$Particles.ArmWiggleFrequency, "[1] Amplitude Min/Max"=<0 1>)
    )
    Index 1 Param 4=&Lerp_Float(
     A=-0.5
     B=0.5
     Alpha=&Waveform("Waveform [1]"=::NewEnumerator0, "[1] Frequency"=$Particles.ArmWiggleFrequency, "[1] Amplitude Min/Max"=<0 1>)
    )
    Index 2 Param 1=&SelectFloatFromArray(
     Array Sampling Mode=::NewEnumerator0
     Direct Array Index=?
     Float Selection Array=DataInterface() {propertyValues {FloatData [0,1]}}
     Random Seed=0
    )
    Index 2 Param 2=&SelectFloatFromArray(
     Array Sampling Mode=::NewEnumerator0
     Direct Array Index=?
     Float Selection Array=DataInterface() {propertyValues {FloatData [0,1]}}
     Random Seed=0
    )
    Module.Index0Float0Write=true
    Module.Index0Float1Write=true
    Module.Index0Float2Write=true
    Module.Index0Float3Write=true
    Module.Index1Float0Write=true
    Module.Index1Float1Write=true
    Module.Index1Float2Write=true
    Module.Index1Float3Write=true
    Module.Index2Float0Write=true
    Module.Index2Float1Write=true
    Module.Index2Float2Write=false
    Module.Index2Float3Write=false
    Module.Index3Float0Write=false
    Module.Index3Float1Write=false
    Module.Index3Float2Write=false
    Module.Index3Float3Write=false
    Write Parameter Index 1=::NewEnumerator1
    Write Parameter Index 2=::NewEnumerator1
    Write Parameter Index 3=::NewEnumerator1
   )
   &MaintainInCameraParticleScale(
    Pixel Depth:NiagaraFloat=&Add_Float(A=$Engine.Owner.LODDistance, B=100)
    Render Target Width:NiagaraFloat=&Multiply_Float(A=2560, B=1)
   ) {enabled=false, module=MaintainInCameraParticleScale001}
   &FN_Sprite_Attributes(
    Camera Offset=50
    Color=&MultiplyLinearColor(
     Linear Color=&MultiplyLinearColor(
      Linear Color=&ColorFromCurve(ColorCurve=DataInterface() {propertyValues {
         AlphaCurve {keys [<at 0.0609 0 RCIM_Cubic>,<at 0.6127 0.3 RCIM_Cubic>]}
         BlueCurve {keys [<at 0.0767 0.6337 RCIM_Cubic>,<at 0.5465 1 RCIM_Cubic>]}
         GreenCurve {keys [<at 0.0767 1 RCIM_Cubic>,<at 0.5465 0.2055 RCIM_Cubic>]}
         RedCurve {keys [<at 0.0767 0 RCIM_Cubic>,<at 0.5465 0 RCIM_Cubic>]}
        }})
      Scale Factor=&MakeVector4FromVectorAndFloat("Vector (RGB)"=&VectorFromFloat(Value=5), "Float (Alpha)"=1)
     )
     Scale Factor=&MakeVector4_FromLinearColor(LinearColor=$Emitter.TODColorMult)
    )
    Create Camera Offset Attribute=true
    Create Color Attribute=true
    "Create Normalized Age "=true
    Create SpriteSize Attribute=true
    SpriteSize=&Multiply_Vector2D_ByFloat(
     Float=&Multiply_Float(
      A=&FloatFromCurve(FloatCurve=DataInterface() {propertyValues {Curve {keys [
           {interpMode=RCIM_Cubic, leaveTangent=14.8021, tangentMode=RCTM_Break, time=0, value=0}
           {arriveTangent=-5.4057, interpMode=RCIM_Cubic, tangentMode=RCTM_Break, time=0.5, value=5}
          ]}}})
      B=&FN_LOD_Distance_Remap(Start Distance=0, End Distance=1000, Minimum Output=0.1, Maximum Output=2, Clamp Output=true)
     )
     Vector2D=<150 150>
    )
    Write to Module Namespace=true
   )
   &FN_Light_Attributes(
    Color=<rgba 0.001 1 0.864 1>
    Create Color Attribute=true
    Create Exponent Attribute=true
    Create Radius Attribute=true
    Enable Particle Light Enable=true
    Exponent=5
    Light Enabled=true
    Radius=500
    Scale Color=5
   ) {enabled=false}
   &FN_Mesh_Attributes(
    Color=&MultiplyLinearColor(
     Linear Color=<rgba 0.0094 4 2.1307 0.1043>
     Scale Factor=&MakeVector4_FromLinearColor(LinearColor=$Emitter.TODColorMult)
    )
    Create Color Attribute=true
    Create Scale Attribute=true
    Scale=&Multiply_VectorByFloat(Vector=<2 1 1>, Float=0.5)
    Write to Module Namespace=true
   )
  ]
  rendererClasses [&NiagaraLightRendererProperties, &NiagaraMeshRendererProperties]
  renderers [
   {rendererClass=&NiagaraLightRendererProperties, rendererIndex=0}
   {rendererClass=&NiagaraMeshRendererProperties, rendererIndex=1}
   {rendererClass=&NiagaraMeshRendererProperties, rendererIndex=2}
  ]
 }
}
summary {
 systemName=NS_Updoot_Launcher_PrjTrail
 emitters [
  {emitterName=Rocket_Trail_Ribbon, rendererClasses [&NiagaraRibbonRendererProperties]}
  {emitterName=Rocket_Trail_Puffy_Smoke, rendererClasses [&NiagaraSpriteRendererProperties]}
  {
   bEnabled=false
   emitterName=Updoot_Arrow
   rendererClasses [
    &NiagaraSpriteRendererProperties
    &NiagaraLightRendererProperties
    &NiagaraMeshRendererProperties
   ]
  }
  {emitterName=Updoot_Sprite, rendererClasses [&NiagaraLightRendererProperties, &NiagaraMeshRendererProperties]}
 ]
 userVariables [User.ProjLifetime {type=NiagaraFloat, defaultValue=10, description=""}]
}
