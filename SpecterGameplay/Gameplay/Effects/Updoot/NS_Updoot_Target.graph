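# Niagara targeting reticle for Updoot, drawing an arrow and four rotated chevrons as camera-scaled sprites, each paired with a soft-glow copy.
# User.LockProgress triggers per-partition timelines, staggered across the chevrons, that drive pivot offset and alpha, masked by spawn group.
# BulletCount feeds a dynamic material parameter, while Cooldown Active switches the visibility tag and sprite scaling.
# A bone-attached body Glow emitter is present but disabled.
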NiagaraSystem /SpecterGameplay/Gameplay/Effects/Updoot/NS_Updoot_Target
dependencies {
 usedDataInterfaces [
  &NiagaraDataInterfaceArrayFloat
  &NiagaraDataInterfaceCurve
  &NiagaraDataInterfaceArrayInt32
  &NiagaraDataInterfaceColorCurve
  &NiagaraDataInterfaceSkeletalMesh
 ]
 usedDynamicInputs []
 usedModules [
  &self:SystemSpawnScript.SystemScriptSource.SystemScriptGraph.NiagaraNodeAssignment_0.SetVariables_#1
  &SystemState
 ]
}
emitters {
 DootArrow_Targeted_Reticle {
  emitterUpdateScript [
   &EmitterState(
    Inactive Response=::NewEnumerator1
    Life Cycle Mode=::NewEnumerator1
    Loop Behavior=::NewEnumerator1
    Loop Duration Mode=::NewEnumerator1
    UseLoopDelay=false
   )
   &SpawnBurst_Instantaneous(Use Spawn Probability=false, UseLoopCountLimit=true)
   &SpawnBurst_Instantaneous(Spawn Group=1, Use Spawn Probability=false, UseLoopCountLimit=true) {module=SpawnBurst_Instantaneous001}
  ]
  particleSpawnScript [
   &PartitionParticles(Number Of Partitions=1)
   &InitializeParticle(
    Position Offset=<0 0 100>
    Sprite Rotation Angle=&SelectFloatFromArray(
     Direct Array Index=$Particles.PartitionParticles.Partition
     Float Selection Array=DataInterface() {propertyValues {FloatData [0,90,180,270]}}
    )
    Sprite Rotation Mode=::NewEnumerator0
    Sprite Size Mode=::NewEnumerator3
    Uniform Sprite Size=25
    UsePositionOffset=false
    Write Material Random=true
   ) {enabled=false}
   &InitializeParticle(
    Position Offset=<0 0 100>
    Sprite Rotation Angle=&SelectFloatFromArray(
     Direct Array Index=$Particles.PartitionParticles.Partition
     Float Selection Array=DataInterface() {propertyValues {FloatData [45,135,225,315]}}
    )
    Sprite Rotation Mode=::NewEnumerator0
    Sprite Size Mode=::NewEnumerator3
    Uniform Sprite Size=25
    UsePositionOffset=false
    Write Material Random=true
   ) {module=InitializeParticle001}
   &ShapeLocation(MaskSpawnGroup=false, Override Seed=false, Rotation Angle=45, Rotation Mode=::NewEnumerator1, Sphere Radius=0, UseOwnerScale=true) {enabled=false}
  ]
  particleUpdateScript [
   &ParticleState(Kill Particles When Lifetime Has Elapsed=false, Let Infinitely Lived Particles Die When Emitter Deactivates=true)
   &ScaleSpriteSize(
    Scale Sprite Size Mode=::NewEnumerator0
    Uniform Scale Factor=&Lerp_Float(A=1, B=3, Alpha=&MakeCustomFloatFromBool(Bool=$User.Cooldown Active, True Float=1, False Float=0))
   ) {enabled=false}
   &SkeletalMeshLocation(
    Bone Sampling Mode=::NewEnumerator4
    MaskSpawnGroup=false
    Orientation Sampling=::NewEnumerator1
    Sampled Position Offset=<16 0 0>
    Skeletal Mesh=$System.Skeletal Mesh
   ) {enabled=false}
   /Niagara/Modules/Update/Camera/MaintainInCameraParticleScale(
    Half FOV:NiagaraFloat=&MakeCustomFloatFromBool(
     Bool=$User.Cooldown Active
     True Float=&Lerp_Float(
      A=55
      B=65
      Alpha=&FloatFromCurve(
       FloatCurve=DataInterface() {propertyValues {Curve {defaultValue=0, keys [{tangentMode=RCTM_User, time=0, value=0}, {tangentMode=RCTM_User, time=1, value=1}]}}}
       CurveIndex=$User.Cooldown Alpha
      )
     )
     False Float=50
    )
    Pixel Depth:NiagaraFloat=&Add_Float(A=$Engine.Owner.LODDistance, B=100)
    Render Target Width:NiagaraFloat=&Multiply_Float(A=2560, B=1)
   ) {enabled=false}
   &MaintainInCameraParticleScale(
    Half FOV:NiagaraFloat=&MakeCustomFloatFromBool(
     Bool=$User.Cooldown Active
     True Float=&Lerp_Float(
      A=55
      B=65
      Alpha=&FloatFromCurve(
       FloatCurve=DataInterface() {propertyValues {Curve {defaultValue=0, keys [{tangentMode=RCTM_User, time=0, value=1}, {tangentMode=RCTM_User, time=1, value=1}]}}}
       CurveIndex=$User.Cooldown Alpha
      )
     )
     False Float=50
    )
    Pixel Depth:NiagaraFloat=&Add_Float(A=$Engine.Owner.LODDistance, B=100)
    Render Target Width:NiagaraFloat=&Multiply_Float(A=2560, B=1)
   ) {module=MaintainInCameraParticleScale001}
   &ScaleSpriteSize(
    Scale Sprite Size Mode=::NewEnumerator0
    Uniform Scale Factor=&Lerp_Float(A=1.25, B=0.5, Alpha=&MakeCustomFloatFromBool(Bool=$User.Cooldown Active, True Float=1, False Float=0))
   ) {enabled=false, module=ScaleSpriteSize001}
   &Timeline(
    Play=&SetBoolByFloatComparison(
     A=&Subtract_Float(A=$User.LockProgress, B=&Multiply_FloatByInt(Float=0.25, Integer=$Particles.PartitionParticles.Partition))
     B=0.175
    )
    Timeline End=0.15
   )
   &DynamicMaterialParameters(
    Index 0 Param 1=&MaskFloatBySpawnGroup(
     Masked Float=&MakeCustomFloatFromBool(
      Bool=&SetBoolByIntComparison(A=$User.BulletCount, B=$Particles.Initial.PartitionParticles.Partition, Comparison Type=::NewEnumerator0)
      True Float=0
      False Float=1
     )
     Passthrough Float=0
    )
    Module.Param0WriteEnabled=true
    Module.Param1WriteEnabled=false
    Module.Param2WriteEnabled=false
    Module.Param3WriteEnabled=false
   )
   /SetVariables(Particles.PivotOffset:Vector2f=&Lerp_Vector2(
     A=<1.3 1.3>
     B=<2.5 2.5>
     Alpha=&FloatFromCurve(
      FloatCurve=DataInterface() {propertyValues {Curve {keys [{interpMode=RCIM_Cubic, tangentMode=RCTM_User, time=0, value=1}, {interpMode=RCIM_Cubic, tangentMode=RCTM_User, time=0.66, value=0}]}}}
      CurveIndex=$Particles.Timeline.NormalizedTime
      Scale Curve=&MaskFloatBySpawnGroup(Masked Float=0, Passthrough Float=1)
     )
    )) {isSetParameters=true, module=SetVariables_#2}
   /SetVariables(Particles.SpriteSize=&Multiply_Vector2D_ByFloat(Float=&Lerp_Float(A=1, B=1, Alpha=0), Vector2D=$Particles.SpriteSize)) {isSetParameters=true, module=SetVariables_#3}
   /SetVariables(Particles.SpriteSize=/Niagara/DynamicInputs/Add/Add_Vector2(
     A:Vector2f=$Particles.SpriteSize
     B:Vector2f=&Vector2DFromFloat(Value=&FloatFromCurve(
       FloatCurve=DataInterface() {propertyValues {Curve {keys [
           {arriveTangent=9.0725, interpMode=RCIM_Cubic, leaveTangent=9.0725, tangentMode=RCTM_User, time=0, value=0}
           {arriveTangent=-3.0559, interpMode=RCIM_Cubic, leaveTangent=-3.0559, tangentMode=RCTM_User, time=0.25, value=1}
           {arriveTangent=-0.0405, interpMode=RCIM_Cubic, leaveTangent=-0.0405, tangentMode=RCTM_User, time=1, value=0}
          ]}}}
       CurveIndex=$System.Timeline001.NormalizedTime
       Scale Curve=50
      ))
    )) {isSetParameters=true, module=SetVariables_#4}
   &ScaleColor(
    Scale Alpha=&FloatFromCurve(
     FloatCurve=DataInterface() {propertyValues {Curve {keys [
         {arriveTangent=5.6222, interpMode=RCIM_Cubic, time=0, value=0}
         {interpMode=RCIM_Cubic, tangentMode=RCTM_User, time=0.2, value=1}
         {arriveTangent=-0.0405, interpMode=RCIM_Cubic, leaveTangent=-0.0405, tangentMode=RCTM_User, time=1, value=1}
        ]}}}
     CurveIndex=$Emitter.LoopedAge
     Scale Curve=&MaskFloatBySpawnGroup(
      Masked Float=1
      Passthrough Float=&FloatFromCurve(
       FloatCurve=DataInterface() {propertyValues {Curve {keys [
           {interpMode=RCIM_Cubic, tangentMode=RCTM_User, time=0, value=0}
           {interpMode=RCIM_Cubic, tangentMode=RCTM_User, time=0.5, value=1}
           {interpMode=RCIM_Cubic, tangentMode=RCTM_User, time=0.75, value=1}
           {tangentMode=RCTM_User, time=1, value=0}
          ]}}}
       CurveIndex=$Particles.Timeline.NormalizedTime
      )
     )
    )
    Scale RGB=&Lerp_Vector(A=<1 1 1>, B=<1 1 1>, Alpha=$System.LockProgress)
    ScaleA=true
    ScaleRGB=true
   )
   /SetVariables(Particles.VisibilityTag=&SelectIntFromArray(
     Direct Array Index=&MakeCustomIntFromBool(Bool=$User.Cooldown Active, True INT=0, False INT=1)
     Int Selection Array=DataInterface() {propertyValues {IntData [1,0]}}
    )) {isSetParameters=true, module=SetVariables_#5}
   /SetVariables(Particles.CooldownColor:LinearColor=&MultiplyLinearColor(
     Linear Color=&ColorFromCurve(
      ColorCurve=DataInterface() {propertyValues {
        AlphaCurve {keys [<at 0 1 RCIM_Cubic>,<at 1 1 RCIM_Cubic>]}
        BlueCurve {keys [<at 0.1098 0.9565 RCIM_Cubic>,<at 0.3882 1 RCIM_Cubic>]}
        GreenCurve {keys [<at 0.1098 2 RCIM_Cubic>,<at 0.3882 1 RCIM_Cubic>]}
        RedCurve {keys [<at 0.1098 0 RCIM_Cubic>,<at 0.3882 1 RCIM_Cubic>]}
       }}
      CurveIndex=$User.Cooldown Alpha
     )
     Scale Factor=&MakeVector4(
      X=1
      Y=1
      Z=1
      W=&Multiply_Float(
       A=&Lerp_Float(
        A=0.5
        B=0.1
        Alpha=&FloatFromCurve(
         FloatCurve=DataInterface() {propertyValues {Curve {
            defaultValue=1
            keys [
             <at 0 0 RCIM_Cubic>
             {arriveTangent=0.2333, interpMode=RCIM_Cubic, leaveTangent=0.2333, tangentMode=RCTM_User, time=0.8, value=0.1}
             {arriveTangent=2.2026, tangentMode=RCTM_User, time=1, value=1}
            ]
           }}}
         CurveIndex=$User.Cooldown Alpha
        )
       )
       B=&SelectFloatFromArray(Direct Array Index=$Particles.PartitionParticles.Partition, Float Selection Array=DataInterface() {propertyValues {FloatData [1,1,1,1]}})
      )
     )
    )) {enabled=false, isSetParameters=true}
   &CameraOffset(Camera Offset Amount=1)
  ]
  rendererClasses [&NiagaraSpriteRendererProperties]
  renderers [
   {rendererClass=&NiagaraSpriteRendererProperties}
   {rendererClass=&NiagaraSpriteRendererProperties, rendererIndex=1}
  ]
 }
 DootArrow_Targeted_Reticle001 {
  emitterUpdateScript [
   &EmitterState(
    Inactive Response=::NewEnumerator1
    Life Cycle Mode=::NewEnumerator1
    Loop Behavior=::NewEnumerator1
    Loop Duration Mode=::NewEnumerator1
    UseLoopDelay=false
   )
   &SpawnBurst_Instantaneous(Use Spawn Probability=false, UseLoopCountLimit=true)
  ]
  particleSpawnScript [
   &PartitionParticles(Number Of Partitions=1)
   &InitializeParticle(
    Position Offset=<0 0 100>
    Sprite Rotation Angle=&SelectFloatFromArray(
     Direct Array Index=$Particles.PartitionParticles.Partition
     Float Selection Array=DataInterface() {propertyValues {FloatData [0,90,180,270]}}
    )
    Sprite Rotation Mode=::NewEnumerator0
    Sprite Size Mode=::NewEnumerator3
    Uniform Sprite Size=25
    UsePositionOffset=false
    Write Material Random=true
   ) {enabled=false}
   &InitializeParticle(
    Position Offset=<0 0 100>
    Sprite Rotation Angle=&SelectFloatFromArray(
     Direct Array Index=$Particles.PartitionParticles.Partition
     Float Selection Array=DataInterface() {propertyValues {FloatData [45,135,225,315]}}
    )
    Sprite Rotation Mode=::NewEnumerator0
    Sprite Size Mode=::NewEnumerator3
    Uniform Sprite Size=24
    UsePositionOffset=false
    Write Material Random=true
   ) {module=InitializeParticle001}
   &ShapeLocation(MaskSpawnGroup=false, Override Seed=false, Rotation Angle=45, Rotation Mode=::NewEnumerator1, Sphere Radius=0, UseOwnerScale=true) {enabled=false}
  ]
  particleUpdateScript [
   &ParticleState(Kill Particles When Lifetime Has Elapsed=false, Let Infinitely Lived Particles Die When Emitter Deactivates=true)
   &SkeletalMeshLocation(
    Bone Sampling Mode=::NewEnumerator4
    MaskSpawnGroup=false
    Orientation Sampling=::NewEnumerator1
    Sampled Position Offset=<16 0 0>
    Skeletal Mesh=$System.Skeletal Mesh
   ) {enabled=false}
   &MaintainInCameraParticleScale(
    Half FOV:NiagaraFloat=&MakeCustomFloatFromBool(
     Bool=$User.Cooldown Active
     True Float=&Lerp_Float(
      A=55
      B=65
      Alpha=&FloatFromCurve(
       FloatCurve=DataInterface() {propertyValues {Curve {defaultValue=0, keys [{tangentMode=RCTM_User, time=0, value=0}, {tangentMode=RCTM_User, time=1, value=1}]}}}
       CurveIndex=$User.Cooldown Alpha
      )
     )
     False Float=50
    )
    Pixel Depth:NiagaraFloat=&Add_Float(A=$Engine.Owner.LODDistance, B=100)
    Render Target Width:NiagaraFloat=&Multiply_Float(A=2560, B=1)
   )
   &Timeline(
    Play=&SetBoolByFloatComparison(
     A=&Subtract_Float(A=$User.LockProgress, B=&Multiply_FloatByInt(Float=0.25, Integer=$Particles.PartitionParticles.Partition))
     B=0.175
    )
    Timeline End=0.15
   )
   &DynamicMaterialParameters(
    Index 0 Param 1=&MaskFloatBySpawnGroup(
     Masked Float=&MakeCustomFloatFromBool(
      Bool=&SetBoolByIntComparison(A=$User.BulletCount, B=$Particles.Initial.PartitionParticles.Partition, Comparison Type=::NewEnumerator0)
      True Float=0
      False Float=1
     )
     Passthrough Float=0
    )
    Module.Param0WriteEnabled=true
    Module.Param1WriteEnabled=false
    Module.Param2WriteEnabled=false
    Module.Param3WriteEnabled=false
   )
   /SetVariables(Particles.PivotOffset:Vector2f=&Lerp_Vector2(
     A=<1.4 1.4>
     B=<2.5 2.5>
     Alpha=&FloatFromCurve(
      FloatCurve=DataInterface() {propertyValues {Curve {keys [{interpMode=RCIM_Cubic, tangentMode=RCTM_User, time=0, value=1}, {interpMode=RCIM_Cubic, tangentMode=RCTM_User, time=0.66, value=0}]}}}
      CurveIndex=$Particles.Timeline.NormalizedTime
      Scale Curve=&MaskFloatBySpawnGroup(Masked Float=0, Passthrough Float=1)
     )
    )) {isSetParameters=true, module=SetVariables_#2}
   /SetVariables(Particles.SpriteSize=&Multiply_Vector2D_ByFloat(Float=&Lerp_Float(A=1, B=1, Alpha=0), Vector2D=$Particles.SpriteSize)) {isSetParameters=true, module=SetVariables_#3}
   /SetVariables(Particles.SpriteSize=&Add_Vector2(
     A:Vector2f=$Particles.SpriteSize
     B:Vector2f=&Vector2DFromFloat(Value=&FloatFromCurve(
       FloatCurve=DataInterface() {propertyValues {Curve {keys [
           {arriveTangent=9.0725, interpMode=RCIM_Cubic, leaveTangent=9.0725, tangentMode=RCTM_User, time=0, value=0}
           {arriveTangent=-3.0559, interpMode=RCIM_Cubic, leaveTangent=-3.0559, tangentMode=RCTM_User, time=0.25, value=1}
           {arriveTangent=-0.0405, interpMode=RCIM_Cubic, leaveTangent=-0.0405, tangentMode=RCTM_User, time=1, value=0}
          ]}}}
       CurveIndex=$System.Timeline001.NormalizedTime
       Scale Curve=50
      ))
    )) {isSetParameters=true, module=SetVariables_#4}
   &ScaleColor(
    Scale Alpha=&FloatFromCurve(
     FloatCurve=DataInterface() {propertyValues {Curve {keys [
         {arriveTangent=5.6222, interpMode=RCIM_Cubic, time=0, value=0}
         {interpMode=RCIM_Cubic, tangentMode=RCTM_User, time=0.2, value=1}
         {arriveTangent=-0.0405, interpMode=RCIM_Cubic, leaveTangent=-0.0405, tangentMode=RCTM_User, time=1, value=1}
        ]}}}
     CurveIndex=$Emitter.LoopedAge
     Scale Curve=&MaskFloatBySpawnGroup(
      Masked Float=1
      Passthrough Float=&FloatFromCurve(
       FloatCurve=DataInterface() {propertyValues {Curve {keys [
           {interpMode=RCIM_Cubic, tangentMode=RCTM_User, time=0, value=0}
           {interpMode=RCIM_Cubic, tangentMode=RCTM_User, time=0.5, value=1}
           {interpMode=RCIM_Cubic, tangentMode=RCTM_User, time=0.75, value=1}
           {tangentMode=RCTM_User, time=1, value=0}
          ]}}}
       CurveIndex=$Particles.Timeline.NormalizedTime
      )
     )
    )
    Scale RGB=&Lerp_Vector(A=<1 1 1>, B=<1 1 1>, Alpha=$System.LockProgress)
    ScaleA=true
    ScaleRGB=true
   )
   /SetVariables(Particles.VisibilityTag=&SelectIntFromArray(
     Direct Array Index=&MakeCustomIntFromBool(Bool=$User.Cooldown Active, True INT=0, False INT=1)
     Int Selection Array=DataInterface() {propertyValues {IntData [1,0]}}
    )) {isSetParameters=true, module=SetVariables_#5}
   /SetVariables(Particles.SoftGlow:LinearColor=<rgba 1.5 1.5 1.5 0.0435>) {isSetParameters=true}
   &CameraOffset(Camera Offset Amount=-2)
  ]
  rendererClasses [&NiagaraSpriteRendererProperties]
  renderers [{rendererClass=&NiagaraSpriteRendererProperties}]
 }
 DootChevrons_Targeted_Reticle {
  emitterUpdateScript [
   &EmitterState(
    Inactive Response=::NewEnumerator1
    Life Cycle Mode=::NewEnumerator1
    Loop Behavior=::NewEnumerator1
    Loop Duration Mode=::NewEnumerator1
    UseLoopDelay=false
   )
   &SpawnBurst_Instantaneous(Spawn Count=4, Use Spawn Probability=false, UseLoopCountLimit=true)
   &SpawnBurst_Instantaneous(Spawn Count=4, Spawn Group=1, Use Spawn Probability=false, UseLoopCountLimit=true) {module=SpawnBurst_Instantaneous001}
  ]
  particleSpawnScript [
   &PartitionParticles(Number Of Partitions=4)
   &InitializeParticle(
    Position Offset=<0 0 100>
    Sprite Rotation Angle=&SelectFloatFromArray(
     Direct Array Index=$Particles.PartitionParticles.Partition
     Float Selection Array=DataInterface() {propertyValues {FloatData [0,90,180,270]}}
    )
    Sprite Rotation Mode=::NewEnumerator0
    Sprite Size Mode=::NewEnumerator3
    Uniform Sprite Size=25
    UsePositionOffset=false
    Write Material Random=true
   ) {enabled=false}
   &InitializeParticle(
    Position Offset=<0 0 100>
    Sprite Rotation Angle=&SelectFloatFromArray(
     Direct Array Index=$Particles.PartitionParticles.Partition
     Float Selection Array=DataInterface() {propertyValues {FloatData [45,135,225,315]}}
    )
    Sprite Rotation Mode=::NewEnumerator0
    Sprite Size Mode=::NewEnumerator3
    Uniform Sprite Size=25
    UsePositionOffset=false
    Write Material Random=true
   ) {module=InitializeParticle001}
   &ShapeLocation(MaskSpawnGroup=false, Override Seed=false, Rotation Angle=45, Rotation Mode=::NewEnumerator1, Sphere Radius=0, UseOwnerScale=true) {enabled=false}
  ]
  particleUpdateScript [
   &ParticleState(Kill Particles When Lifetime Has Elapsed=false, Let Infinitely Lived Particles Die When Emitter Deactivates=true)
   &ScaleSpriteSize(
    Scale Sprite Size Mode=::NewEnumerator0
    Uniform Scale Factor=&Lerp_Float(A=1, B=3, Alpha=&MakeCustomFloatFromBool(Bool=$User.Cooldown Active, True Float=1, False Float=0))
   ) {enabled=false}
   &SkeletalMeshLocation(
    Bone Sampling Mode=::NewEnumerator4
    MaskSpawnGroup=false
    Orientation Sampling=::NewEnumerator1
    Sampled Position Offset=<16 0 0>
    Skeletal Mesh=$System.Skeletal Mesh
   ) {enabled=false}
   &MaintainInCameraParticleScale(
    Half FOV:NiagaraFloat=&MakeCustomFloatFromBool(
     Bool=$User.Cooldown Active
     True Float=&Lerp_Float(
      A=55
      B=65
      Alpha=&FloatFromCurve(
       FloatCurve=DataInterface() {propertyValues {Curve {defaultValue=0, keys [{tangentMode=RCTM_User, time=0, value=0}, {tangentMode=RCTM_User, time=1, value=1}]}}}
       CurveIndex=$User.Cooldown Alpha
      )
     )
     False Float=50
    )
    Pixel Depth:NiagaraFloat=&Add_Float(A=$Engine.Owner.LODDistance, B=100)
    Render Target Width:NiagaraFloat=&Multiply_Float(A=2560, B=1)
   ) {enabled=false}
   &MaintainInCameraParticleScale(
    Half FOV:NiagaraFloat=&MakeCustomFloatFromBool(
     Bool=$User.Cooldown Active
     True Float=&Lerp_Float(
      A=55
      B=65
      Alpha=&FloatFromCurve(
       FloatCurve=DataInterface() {propertyValues {Curve {defaultValue=0, keys [{tangentMode=RCTM_User, time=0, value=1}, {tangentMode=RCTM_User, time=1, value=1}]}}}
       CurveIndex=$User.Cooldown Alpha
      )
     )
     False Float=50
    )
    Pixel Depth:NiagaraFloat=&Add_Float(A=$Engine.Owner.LODDistance, B=100)
    Render Target Width:NiagaraFloat=&Multiply_Float(A=2560, B=1)
   ) {module=MaintainInCameraParticleScale001}
   &ScaleSpriteSize(
    Scale Sprite Size Mode=::NewEnumerator0
    Uniform Scale Factor=&Lerp_Float(A=1.25, B=0.5, Alpha=&MakeCustomFloatFromBool(Bool=$User.Cooldown Active, True Float=1, False Float=0))
   ) {enabled=false, module=ScaleSpriteSize001}
   &Timeline(
    Play=&SetBoolByFloatComparison(
     A=&Subtract_Float(A=$User.LockProgress, B=&Multiply_FloatByInt(Float=0.25, Integer=$Particles.PartitionParticles.Partition))
     B=0.175
    )
    Timeline End=0.15
   )
   &DynamicMaterialParameters(
    Index 0 Param 1=&MaskFloatBySpawnGroup(
     Masked Float=&MakeCustomFloatFromBool(
      Bool=&SetBoolByIntComparison(A=$User.BulletCount, B=$Particles.Initial.PartitionParticles.Partition, Comparison Type=::NewEnumerator0)
      True Float=0
      False Float=1
     )
     Passthrough Float=0
    )
    Module.Param0WriteEnabled=true
    Module.Param1WriteEnabled=false
    Module.Param2WriteEnabled=false
    Module.Param3WriteEnabled=false
   )
   /SetVariables(Particles.PivotOffset:Vector2f=&Lerp_Vector2(
     A=<1.3 1.3>
     B=<2.5 2.5>
     Alpha=&FloatFromCurve(
      FloatCurve=DataInterface() {propertyValues {Curve {keys [{interpMode=RCIM_Cubic, tangentMode=RCTM_User, time=0, value=1}, {interpMode=RCIM_Cubic, tangentMode=RCTM_User, time=0.66, value=0}]}}}
      CurveIndex=$Particles.Timeline.NormalizedTime
      Scale Curve=&MaskFloatBySpawnGroup(Masked Float=0, Passthrough Float=1)
     )
    )) {isSetParameters=true, module=SetVariables_#2}
   /SetVariables(Particles.SpriteSize=&Multiply_Vector2D_ByFloat(Float=&Lerp_Float(A=1, B=1, Alpha=0), Vector2D=$Particles.SpriteSize)) {isSetParameters=true, module=SetVariables_#3}
   /SetVariables(Particles.SpriteSize=&Add_Vector2(
     A:Vector2f=$Particles.SpriteSize
     B:Vector2f=&Vector2DFromFloat(Value=&FloatFromCurve(
       FloatCurve=DataInterface() {propertyValues {Curve {keys [
           {arriveTangent=9.0725, interpMode=RCIM_Cubic, leaveTangent=9.0725, tangentMode=RCTM_User, time=0, value=0}
           {arriveTangent=-3.0559, interpMode=RCIM_Cubic, leaveTangent=-3.0559, tangentMode=RCTM_User, time=0.25, value=1}
           {arriveTangent=-0.0405, interpMode=RCIM_Cubic, leaveTangent=-0.0405, tangentMode=RCTM_User, time=1, value=0}
          ]}}}
       CurveIndex=$System.Timeline001.NormalizedTime
       Scale Curve=50
      ))
    )) {enabled=false, isSetParameters=true, module=SetVariables_#4}
   &ScaleColor(
    Scale Alpha=&FloatFromCurve(
     FloatCurve=DataInterface() {propertyValues {Curve {keys [
         {arriveTangent=5.6222, interpMode=RCIM_Cubic, time=0, value=0}
         {interpMode=RCIM_Cubic, tangentMode=RCTM_User, time=0.2, value=1}
         {arriveTangent=-0.0405, interpMode=RCIM_Cubic, leaveTangent=-0.0405, tangentMode=RCTM_User, time=1, value=1}
        ]}}}
     CurveIndex=$Emitter.LoopedAge
     Scale Curve=&MaskFloatBySpawnGroup(
      Masked Float=1
      Passthrough Float=&FloatFromCurve(
       FloatCurve=DataInterface() {propertyValues {Curve {keys [
           {interpMode=RCIM_Cubic, tangentMode=RCTM_User, time=0, value=0}
           {interpMode=RCIM_Cubic, tangentMode=RCTM_User, time=0.5, value=1}
           {interpMode=RCIM_Cubic, tangentMode=RCTM_User, time=0.75, value=1}
           {tangentMode=RCTM_User, time=1, value=0}
          ]}}}
       CurveIndex=$Particles.Timeline.NormalizedTime
      )
     )
    )
    Scale RGB=&Lerp_Vector(A=<1 1 1>, B=<1 1 1>, Alpha=$System.LockProgress)
    ScaleA=true
    ScaleRGB=true
   )
   /SetVariables(Particles.VisibilityTag=&SelectIntFromArray(
     Direct Array Index=&MakeCustomIntFromBool(Bool=$User.Cooldown Active, True INT=0, False INT=1)
     Int Selection Array=DataInterface() {propertyValues {IntData [1,0]}}
    )) {isSetParameters=true, module=SetVariables_#5}
   /SetVariables(Particles.CooldownColor:LinearColor=&MultiplyLinearColor(
     Linear Color=&ColorFromCurve(
      ColorCurve=DataInterface() {propertyValues {
        AlphaCurve {keys [<at 0 1 RCIM_Cubic>,<at 1 1 RCIM_Cubic>]}
        BlueCurve {keys [<at 0.1098 0.9565 RCIM_Cubic>,<at 0.3882 1 RCIM_Cubic>]}
        GreenCurve {keys [<at 0.1098 2 RCIM_Cubic>,<at 0.3882 1 RCIM_Cubic>]}
        RedCurve {keys [<at 0.1098 0 RCIM_Cubic>,<at 0.3882 1 RCIM_Cubic>]}
       }}
      CurveIndex=$User.Cooldown Alpha
     )
     Scale Factor=&MakeVector4(
      X=1
      Y=1
      Z=1
      W=&Multiply_Float(
       A=&Lerp_Float(
        A=0.5
        B=0.1
        Alpha=&FloatFromCurve(
         FloatCurve=DataInterface() {propertyValues {Curve {
            defaultValue=1
            keys [
             <at 0 0 RCIM_Cubic>
             {arriveTangent=0.2333, interpMode=RCIM_Cubic, leaveTangent=0.2333, tangentMode=RCTM_User, time=0.8, value=0.1}
             {arriveTangent=2.2026, tangentMode=RCTM_User, time=1, value=1}
            ]
           }}}
         CurveIndex=$User.Cooldown Alpha
        )
       )
       B=&SelectFloatFromArray(Direct Array Index=$Particles.PartitionParticles.Partition, Float Selection Array=DataInterface() {propertyValues {FloatData [0,1,1,1]}})
      )
     )
    )) {enabled=false, isSetParameters=true}
   &CameraOffset() {enabled=false}
   &ScaleColor(
    Scale Alpha=&SelectFloatFromArray(Direct Array Index=$Particles.PartitionParticles.Partition, Float Selection Array=DataInterface() {propertyValues {FloatData [0,1,1,1]}})
    ScaleA=true
    ScaleRGB=false
   ) {module=ScaleColor001}
  ]
  rendererClasses [&NiagaraSpriteRendererProperties]
  renderers [
   {rendererClass=&NiagaraSpriteRendererProperties}
   {rendererClass=&NiagaraSpriteRendererProperties, rendererIndex=1}
  ]
 }
 DootChevrons_Targeted_Reticle001 {
  emitterUpdateScript [
   &EmitterState(
    Inactive Response=::NewEnumerator1
    Life Cycle Mode=::NewEnumerator1
    Loop Behavior=::NewEnumerator1
    Loop Duration Mode=::NewEnumerator1
    UseLoopDelay=false
   )
   &SpawnBurst_Instantaneous(Spawn Count=4, Use Spawn Probability=false, UseLoopCountLimit=true)
  ]
  particleSpawnScript [
   &PartitionParticles(Number Of Partitions=4)
   &InitializeParticle(
    Position Offset=<0 0 100>
    Sprite Rotation Angle=&SelectFloatFromArray(
     Direct Array Index=$Particles.PartitionParticles.Partition
     Float Selection Array=DataInterface() {propertyValues {FloatData [0,90,180,270]}}
    )
    Sprite Rotation Mode=::NewEnumerator0
    Sprite Size Mode=::NewEnumerator3
    Uniform Sprite Size=25
    UsePositionOffset=false
    Write Material Random=true
   ) {enabled=false}
   &InitializeParticle(
    Position Offset=<0 0 100>
    Sprite Rotation Angle=&SelectFloatFromArray(
     Direct Array Index=$Particles.PartitionParticles.Partition
     Float Selection Array=DataInterface() {propertyValues {FloatData [45,135,225,315]}}
    )
    Sprite Rotation Mode=::NewEnumerator0
    Sprite Size Mode=::NewEnumerator3
    Uniform Sprite Size=25
    UsePositionOffset=false
    Write Material Random=true
   ) {module=InitializeParticle001}
   &ShapeLocation(MaskSpawnGroup=false, Override Seed=false, Rotation Angle=45, Rotation Mode=::NewEnumerator1, Sphere Radius=0, UseOwnerScale=true) {enabled=false}
  ]
  particleUpdateScript [
   &ParticleState(Kill Particles When Lifetime Has Elapsed=false, Let Infinitely Lived Particles Die When Emitter Deactivates=true)
   &SkeletalMeshLocation(
    Bone Sampling Mode=::NewEnumerator4
    MaskSpawnGroup=false
    Orientation Sampling=::NewEnumerator1
    Sampled Position Offset=<16 0 0>
    Skeletal Mesh=$System.Skeletal Mesh
   ) {enabled=false}
   &MaintainInCameraParticleScale(
    Half FOV:NiagaraFloat=&MakeCustomFloatFromBool(
     Bool=$User.Cooldown Active
     True Float=&Lerp_Float(
      A=55
      B=65
      Alpha=&FloatFromCurve(
       FloatCurve=DataInterface() {propertyValues {Curve {defaultValue=0, keys [{tangentMode=RCTM_User, time=0, value=0}, {tangentMode=RCTM_User, time=1, value=1}]}}}
       CurveIndex=$User.Cooldown Alpha
      )
     )
     False Float=50
    )
    Pixel Depth:NiagaraFloat=&Add_Float(A=$Engine.Owner.LODDistance, B=100)
    Render Target Width:NiagaraFloat=&Multiply_Float(A=2560, B=1)
   )
   &Timeline(
    Play=&SetBoolByFloatComparison(
     A=&Subtract_Float(A=$User.LockProgress, B=&Multiply_FloatByInt(Float=0.25, Integer=$Particles.PartitionParticles.Partition))
     B=0.175
    )
    Timeline End=0.15
   )
   &DynamicMaterialParameters(
    Index 0 Param 1=&MaskFloatBySpawnGroup(
     Masked Float=&MakeCustomFloatFromBool(
      Bool=&SetBoolByIntComparison(A=$User.BulletCount, B=$Particles.Initial.PartitionParticles.Partition, Comparison Type=::NewEnumerator0)
      True Float=0
      False Float=1
     )
     Passthrough Float=0
    )
    Module.Param0WriteEnabled=true
    Module.Param1WriteEnabled=false
    Module.Param2WriteEnabled=false
    Module.Param3WriteEnabled=false
   )
   /SetVariables(Particles.PivotOffset:Vector2f=&Lerp_Vector2(
     A=<1.3 1.3>
     B=<2.5 2>
     Alpha=&FloatFromCurve(
      FloatCurve=DataInterface() {propertyValues {Curve {keys [{interpMode=RCIM_Cubic, tangentMode=RCTM_User, time=0, value=1}, {interpMode=RCIM_Cubic, tangentMode=RCTM_User, time=0.66, value=0}]}}}
      CurveIndex=$Particles.Timeline.NormalizedTime
      Scale Curve=&MaskFloatBySpawnGroup(Masked Float=0, Passthrough Float=1)
     )
    )) {isSetParameters=true, module=SetVariables_#2}
   /SetVariables(Particles.SpriteSize=&Multiply_Vector2D_ByFloat(Float=&Lerp_Float(A=1, B=1, Alpha=0), Vector2D=$Particles.SpriteSize)) {isSetParameters=true, module=SetVariables_#3}
   /SetVariables(Particles.SpriteSize=&Add_Vector2(
     A:Vector2f=$Particles.SpriteSize
     B:Vector2f=&Vector2DFromFloat(Value=&FloatFromCurve(
       FloatCurve=DataInterface() {propertyValues {Curve {keys [
           {arriveTangent=9.0725, interpMode=RCIM_Cubic, leaveTangent=9.0725, tangentMode=RCTM_User, time=0, value=0}
           {arriveTangent=-3.0559, interpMode=RCIM_Cubic, leaveTangent=-3.0559, tangentMode=RCTM_User, time=0.25, value=1}
           {arriveTangent=-0.0405, interpMode=RCIM_Cubic, leaveTangent=-0.0405, tangentMode=RCTM_User, time=1, value=0}
          ]}}}
       CurveIndex=$System.Timeline001.NormalizedTime
       Scale Curve=50
      ))
    )) {isSetParameters=true, module=SetVariables_#4}
   &ScaleColor(
    Scale Alpha=&FloatFromCurve(
     FloatCurve=DataInterface() {propertyValues {Curve {keys [
         {arriveTangent=5.6222, interpMode=RCIM_Cubic, time=0, value=0}
         {interpMode=RCIM_Cubic, tangentMode=RCTM_User, time=0.2, value=1}
         {arriveTangent=-0.0405, interpMode=RCIM_Cubic, leaveTangent=-0.0405, tangentMode=RCTM_User, time=1, value=1}
        ]}}}
     CurveIndex=$Emitter.LoopedAge
     Scale Curve=&MaskFloatBySpawnGroup(
      Masked Float=1
      Passthrough Float=&FloatFromCurve(
       FloatCurve=DataInterface() {propertyValues {Curve {keys [
           {interpMode=RCIM_Cubic, tangentMode=RCTM_User, time=0, value=0}
           {interpMode=RCIM_Cubic, tangentMode=RCTM_User, time=0.5, value=1}
           {interpMode=RCIM_Cubic, tangentMode=RCTM_User, time=0.75, value=1}
           {tangentMode=RCTM_User, time=1, value=0}
          ]}}}
       CurveIndex=$Particles.Timeline.NormalizedTime
      )
     )
    )
    Scale RGB=&Lerp_Vector(A=<1 1 1>, B=<1 1 1>, Alpha=$System.LockProgress)
    ScaleA=true
    ScaleRGB=true
   )
   /SetVariables(Particles.VisibilityTag=&SelectIntFromArray(
     Direct Array Index=&MakeCustomIntFromBool(Bool=$User.Cooldown Active, True INT=0, False INT=1)
     Int Selection Array=DataInterface() {propertyValues {IntData [1,0]}}
    )) {isSetParameters=true, module=SetVariables_#5}
   /SetVariables(Particles.SoftGlow:LinearColor=<rgba 1.5 1.5 1.5 0.0435>) {isSetParameters=true}
   &CameraOffset(Camera Offset Amount=-2)
  ]
  rendererClasses [&NiagaraSpriteRendererProperties]
  renderers [{rendererClass=&NiagaraSpriteRendererProperties}]
 }
 Glow {
  bEnabled=false
  emitterUpdateScript [
   &EmitterState(
    Inactive Response=::NewEnumerator1
    Life Cycle Mode=::NewEnumerator1
    Loop Behavior=::NewEnumerator1
    Loop Delay=0.115
    Loop Duration Mode=::NewEnumerator1
    UseLoopDelay=true
   )
   &SpawnBurst_Instantaneous(Spawn Count=18, Use Spawn Probability=false, UseLoopCountLimit=false)
  ]
  particleSpawnScript [
   &PartitionParticles(Number Of Partitions=18)
   &InitializeParticle(
    Color=<rgba 0 0.875 0.6597 0.3478>
    Lifetime=1.15
    Sprite Rotation Mode=::NewEnumerator0
    Sprite Size Mode=::NewEnumerator3
    Uniform Sprite Size=120
    UsePositionOffset=false
    Write Material Random=true
   ) {module=InitializeParticle001}
  ]
  particleUpdateScript [
   &ParticleState(Kill Particles When Lifetime Has Elapsed=false, Let Infinitely Lived Particles Die When Emitter Deactivates=true)
   &CameraOffset(Camera Offset Amount=50)
   &MaintainInCameraParticleScale(
    Pixel Depth:NiagaraFloat=&Add_Float(A=$Engine.Owner.LODDistance, B=100)
    Render Target Width:NiagaraFloat=&Multiply_Float(A=2560, B=1)
   )
   &ScaleSpriteSize(
    Scale Sprite Size Mode=::NewEnumerator0
    Uniform Scale Factor=&Waveform(
     Global Phase=&Add_Float(A=$Emitter.Age, B=&RandomRangeFloat(Minimum=0, Maximum=5))
     "Waveform [1]"=::NewEnumerator0
     "[1] Frequency"=1
     "[1] Amplitude Min/Max"=<0.85 1>
    )
   ) {enabled=false}
   &SkeletalMeshLocation(
    Bone / Socket Index=$Particles.PartitionParticles.Partition
    Bone Sampling Mode=::NewEnumerator4
    MaskSpawnGroup=false
    Skeletal Mesh=DataInterface() {propertyValues {
      DefaultMesh=null
      ExcludeBoneName=root
      OptionalComponentClass=null
      PreviewMesh=/Game/Characters/Player/Male/Medium/Bodies/M_Med_Soldier_01/Meshes/M_Med_Soldier_01
      SkinningMode=SkinOnTheFly
      SoftSourceActor=null
      SourceMode=Default
      UvSetIndex=0
      WholeMeshLOD=-1
      bExcludeBone=false
      bReadDeformedGeometry=true
      bRequireCurrentFrameData=true
      ComponentTags []
      FilteredBones [head,upperarm_l,upperarm_r,lowerarm_l,lowerarm_r,hand_l,hand_r,spine_05,spine_03]
      FilteredSockets []
      MeshUserParameter {parameter {name=None, typeDefHandle {registeredTypeIndex=1}, varData []}}
      SamplingRegions []
     }}
   ) {module=SkeletalMeshLocation001}
   &ScaleColor(
    Scale Alpha=0.1
    Scale RGB=&VectorFromFloat(Value=&SelectFloatFromArray(Direct Array Index=$Particles.PartitionParticles.Partition, Float Selection Array=DataInterface() {propertyValues {FloatData [1,1,1,1,5,5]}}))
    ScaleA=true
    ScaleRGB=true
   ) {enabled=false}
   &DynamicMaterialParameters(
    Index 0 Param 1=&SelectFloatFromArray(Direct Array Index=$Particles.SpawnGroup, Float Selection Array=DataInterface() {propertyValues {FloatData [5,5,5,5,8,8]}})
    Index 0 Param 3=1
    Module.Param0WriteEnabled=true
    Module.Param1WriteEnabled=true
    Module.Param2WriteEnabled=true
    Module.Param3WriteEnabled=true
   )
   &ScaleColor(
    Scale Alpha=&FloatFromCurve(FloatCurve=DataInterface() {propertyValues {Curve {
        defaultValue=0
        keys [{interpMode=RCIM_Cubic, tangentMode=RCTM_User, time=0, value=1}, {interpMode=RCIM_Cubic, tangentMode=RCTM_User, time=1, value=0}]
       }}})
    Scale RGB=&VectorFromFloat(Value=&FloatFromCurve(FloatCurve=DataInterface() {propertyValues {Curve {
         defaultValue=1
         keys [
          {arriveTangent=34.1211, interpMode=RCIM_Cubic, leaveTangent=34.1211, tangentMode=RCTM_User, time=0, value=0}
          {arriveTangent=-2.5249, interpMode=RCIM_Cubic, leaveTangent=-2.5249, tangentMode=RCTM_User, time=0.1408, value=0.9847}
          {arriveTangent=-0.0405, interpMode=RCIM_Cubic, leaveTangent=-0.0405, tangentMode=RCTM_User, time=1, value=0}
         ]
        }}}))
    ScaleA=true
    ScaleRGB=true
   ) {enabled=false, module=ScaleColor001}
   &ScaleColor(Scale RGB=&VectorFromFloat(Value=0.5), ScaleA=false, ScaleRGB=true) {enabled=false, module=ScaleColor002}
   /SetVariables(Particles.AirGlowColor:LinearColor=&MultiplyLinearColor(Linear Color=$Particles.Color, Scale Factor=<1 1 1 0.175>)) {isSetParameters=true}
   /SetVariables(Particles.VisibilityTag=&SelectIntFromArray(
     Direct Array Index=&MakeCustomIntFromBool(Bool=$User.Cooldown Active, True INT=0, False INT=1)
     Int Selection Array=DataInterface() {propertyValues {IntData [1,0]}}
    )) {isSetParameters=true}
  ]
  rendererClasses [&NiagaraSpriteRendererProperties]
  renderers [{rendererClass=&NiagaraSpriteRendererProperties}]
 }
}
summary {
 systemName=NS_Updoot_Target
 emitters [
  {emitterName=DootChevrons_Targeted_Reticle, rendererClasses [&NiagaraSpriteRendererProperties]}
  {bEnabled=false, emitterName=Glow, rendererClasses [&NiagaraSpriteRendererProperties]}
  {emitterName=DootArrow_Targeted_Reticle, rendererClasses [&NiagaraSpriteRendererProperties]}
  {emitterName=DootChevrons_Targeted_Reticle001, rendererClasses [&NiagaraSpriteRendererProperties]}
  {emitterName=DootArrow_Targeted_Reticle001, rendererClasses [&NiagaraSpriteRendererProperties]}
 ]
 userVariables [
  User.BulletCount {type=NiagaraInt32, defaultValue=4}
  User.Cooldown Active {type=NiagaraBool}
  User.Cooldown Alpha {type=NiagaraFloat, defaultValue=1, description=""}
  User.IsLargeBody {type=NiagaraBool}
  User.LockProgress {type=NiagaraFloat, defaultValue=1, description=""}
 ]
}
