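# Material function that computes world position offset to animate the Updoot sprite effect as a character-like figure.
# Vertex color channels build body-region masks, while two dynamic parameters supply arm, leg, waist, head and face motion controls.
# Masked offsets combine with a RotateVertexAboutAxis rotation around the object pivot, borrowed from the Inferno character's FX folder, into an XYZ output.
# Same-purpose scalar parameters, a Blink parameter and an Is Face switch are present but unconnected.
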MaterialFunction /SpecterGameplay/Gameplay/Effects/Updoot/MF_Sprite_Updoot_WPO
graph {nodes [
  n1 = VertexColor()
  n2 = Multiply(A=@n1.R, B=@n1.R)
  n3 = Saturate(None=Step(Y=OneMinus(None=@n2), X=Multiply() {constB=0.6}) {constX=0.44})
  n4 = NamedRerouteUsage()
  n5 = NamedRerouteUsage()
  n6 = NamedRerouteUsage()
  n7 = Multiply(A=Multiply(A=NamedRerouteUsage(), B=Multiply(A=@n5) {constB=5}), B=@n6)
  n8 = NamedRerouteUsage()
  n9 = NamedRerouteUsage()
  n10 = Multiply(A=NamedRerouteUsage()) {constB=5}
  n11 = Multiply(A=@n8, B=Multiply(A=@n6, B=Multiply(A=@n9, B=@n10)))
  n12 = Power(Base=OneMinus(None=VertexColor().R)) {constExponent=5}
  n13 = NamedRerouteUsage()
  n14 = MaterialFunctionCall!(
   RotationAngle=Multiply(A=NamedRerouteUsage()) {constB=0.15}
   NormalizedRotationAxis=Normalize(VectorInput=VectorParameter() {group=WPO - Waist Twist, parameterName=Rotation Axis, defaultValue=<rgba 0 0 1 1>}.RGB)
   PivotPoint=MaterialFunctionCall() {materialFunction=&ObjectPivotPoint}.Object Pivot Location
   Position
   Normal
   TangentU
   TangentV
  ) {materialFunction=/Game/Characters/Player/Male/Medium/Bodies/M_MED_Inferno/FX/Mat/RotateVertexAboutAxis}
  n15 = MaterialFunctionCall(Color=Multiply(A=NamedRerouteUsage(), B=@n14.WPO)) {materialFunction=&SplitComponents}
  n16 = Add(A=Multiply(A=@n12, B=Multiply(A=@n13) {constB=3}), B=@n15.B)
  n17 = Add(A=Multiply(A=@n1.B, B=Multiply(A=@n4) {constB=4}) {constB=4}, B=Add(A=Multiply(A=@n7, B=@n8), B=Add(A=@n11, B=@n16)))
  n18 = Step(Y=@n2, X=Multiply() {constB=0.6}) {constX=0.44}
  Add(A=Add(A=@n3, B=@n17), B=OneMinus(None=Multiply(A=@n18) {constB=2}))
  Add(A=@n18, B=Subtract(A=@n3, B=@n17))
  n19 = NamedRerouteUsage()
  n20 = Multiply(A=NamedRerouteUsage()) {constB=3}
  n21 = OneMinus(None=NamedRerouteUsage())
  n22 = OneMinus(None=Multiply(A=NamedRerouteUsage()) {constB=3})
  n23 = Multiply(A=NamedRerouteUsage(), B=Add(A=Multiply(A=@n19, B=@n20), B=Multiply(A=@n21, B=@n22)))
  n24 = NamedRerouteUsage()
  n25 = Add(A=Multiply(A=@n12, B=Multiply(A=@n24) {constB=3}), B=@n15.R)
  FunctionOutput(None=MaterialFunctionCall(R=Add(A=@n23, B=@n25), G=@n15.G, B=@n17) {materialFunction=&AppendMany}.RGB)
  n26 = VertexColor()
  NamedRerouteDeclaration(None=Saturate(None=Power(Base=@n26.G) {constExponent=3})) {name=Left Mask, nodeColor=<rgba 0 1 0.4141 1>}
  n27 = OneMinus(None=@n26.G)
  NamedRerouteDeclaration(None=Saturate(None=Power(Base=@n27) {constExponent=3})) {name=Right Mask, nodeColor=<rgba 1 0.7266 0 1>}
  n28 = DynamicParameter() {parameterIndex=1, defaultValue=<rgba 0 0 0 0>, paramNames [Left Arm Wiggle, Right Arm Wiggle, Left Leg Wiggle, Right Leg Wiggle]}
  NamedRerouteDeclaration(None=@n28.Left Arm Wiggle) {name=Left Arm Wiggle, nodeColor=<rgba 1 0.0234 0 1>}
  NamedRerouteDeclaration(None=@n28.Right Arm Wiggle) {name=Right Arm Wiggle, nodeColor=<rgba 0 0.2969 1 1>}
  NamedRerouteDeclaration(None=@n28.Left Leg Wiggle) {name=Left Leg Wiggle, nodeColor=<rgba 0.6172 1 0 1>}
  NamedRerouteDeclaration(None=@n28.Right Leg Wiggle) {name=Right Leg Wiggle, nodeColor=<rgba 1 0 0.9375 1>}
  n29 = DynamicParameter() {defaultValue=<rgba 0 0 0 0>, paramNames [Waist Twist, Head Squash, Head Wiggle, Face Movement]}
  NamedRerouteDeclaration(None=@n29.Waist Twist) {name=Waist Twist, nodeColor=<rgba 0 1 0.7422 1>}
  NamedRerouteDeclaration(None=@n29.Head Squash) {name=Head Squash, nodeColor=<rgba 1 0.4219 0 1>}
  NamedRerouteDeclaration(None=@n29.Head Wiggle) {name=Head Wiggle, nodeColor=<rgba 0.1016 0 1 1>}
  NamedRerouteDeclaration(None=@n29.Face Movement) {name=Face Movement, nodeColor=<rgba 0.2187 1 0 1>}
  n30 = VertexColor()
  NamedRerouteDeclaration(None=Power(Base=@n30.R) {constExponent=10}) {name=Bottom Mask, nodeColor=<rgba 0 0.8125 1 1>}
  n31 = VertexColor()
  NamedRerouteDeclaration(None=Power(Base=@n31.R) {constExponent=20}) {name=Feet Mask, nodeColor=<rgba 0.0781 0 1 1>}
  NamedRerouteDeclaration(None=Transform(None=@n14.Normal) {transformSourceType=TRANSFORMSOURCE_World, transformType=TRANSFORM_Tangent}) {name=WPO Normal, nodeColor=<rgba 0.7812 0 1 1>}
  NamedRerouteDeclaration(None=Step(Y=@n31.G) {constX=0.5}) {name=Half Mask, nodeColor=<rgba 0.2422 1 0 1>}
  n32 = Sine(None=Power(Base=VertexColor().R) {constExponent=2.8})
  NamedRerouteDeclaration(None=Saturate(None=Power(Base=@n32))) {name=Center Mask, nodeColor=<rgba 1 0.7266 0 1>}
  n33 = VertexColor()
  n34 = OneMinus(None=SphereMask(A=AppendVector(A=@n33.R, B=@n33.R), B=Constant() {r=0.5}) {attenuationRadius=0.5, hardnessPercent=0})
  NamedRerouteDeclaration(None=Saturate(None=Power(Base=@n34))) {name=Sphere Mask, nodeColor=<rgba 0.1016 1 0 1>}
  OneMinus(None=Sine(None=@n1.R))
  OneMinus(None=@n30.R)
  ScalarParameter() {group=WPO - Waist Twist, parameterName=Waist Wiggle, sliderMax=1, sliderMin=-1, sortPriority=1}
  ScalarParameter() {group=WPO, parameterName=Right Foot Wiggle, sliderMax=1, sliderMin=-1, sortPriority=0}
  ScalarParameter() {group=WPO, parameterName=Left Foot Wiggle, sliderMax=1, sliderMin=-1, sortPriority=0}
  ScalarParameter() {group=WPO - Head Wiggle, parameterName=Head Wiggle, sliderMax=1, sliderMin=-1}
  ScalarParameter() {defaultValue=0.0341, group=WPO - Head Wiggle, parameterName=Head Squash, sliderMax=1, sliderMin=-1}
  ScalarParameter() {group=WPO, parameterName=Face Movement, sliderMax=1}
  ScalarParameter() {group=WPO, parameterName=Right Arm Wiggle, sliderMax=1, sliderMin=-1, sortPriority=0}
  ScalarParameter() {group=WPO, parameterName=Left Arm Wiggle, sliderMax=1, sliderMin=-1, sortPriority=0}
  ScalarParameter() {defaultValue=1, parameterName=Blink, sliderMax=1}
  StaticSwitch(Value=StaticBoolParameter() {defaultValue=true, group=WPO, parameterName=Is Face})
 ]}
